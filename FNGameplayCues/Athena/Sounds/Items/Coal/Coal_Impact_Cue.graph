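# Sound cue for the impact of the Coal item in Athena, stored with the gameplay cue item sounds.
# A Random node picks one of four equally weighted impact wave variations.
# It plays in the Fort_Impacts sound class, with inline attenuation overriding the referenced close-explosion attenuation asset, adding distance falloff and a low-pass filter that muffles distant impacts.
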SoundCue /FNGameplayCues/Athena/Sounds/Items/Coal/Coal_Impact_Cue
root=Random_0
nodes [
 {id=Random_0, children [WavePlayer_1,WavePlayer_2,WavePlayer_3,WavePlayer_4], properties {weights [1,1,1,1]}}
 {id=WavePlayer_1, properties {soundWaveAssetPtr=&/Weapon_Item_Coal_Impact_01}}
 {id=WavePlayer_2, properties {soundWaveAssetPtr=&/Weapon_Item_Coal_Impact_02}}
 {id=WavePlayer_3, properties {soundWaveAssetPtr=&/Weapon_Item_Coal_Impact_03}}
 {id=WavePlayer_4, properties {soundWaveAssetPtr=&/Weapon_Item_Coal_Impact_04}}
]
properties {
 attenuationSettings=/Game/Sounds/Attenuation/Fort_Explosion_Close_Athena_Att
 bOverrideAttenuation=true
 maxDistance=4600
 soundClassObject=&Fort_Impacts
 volumeMultiplier=1
 attenuationOverrides {bAttenuateWithLPF=true, falloffDistance=4000, lPFFrequencyAtMax=1000, lPFRadiusMax=2500, lPFRadiusMin=600, attenuationShapeExtents {x=600}}
}
